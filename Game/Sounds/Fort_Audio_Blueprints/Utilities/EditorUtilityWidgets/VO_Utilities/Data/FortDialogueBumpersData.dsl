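; FortDialogueBumpersData is a data Blueprint for the VO editor utility widgets, exposing a map of dialogue bumpers.
; GetBumperKeys returns all keys of the Dialogue Bumpers map.
; GetBumpers converts a string key to a Name, looks it up in the map, and returns whether it was found along with the matching value.
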(fn GetBumperKeys
  (bind _keys (Utilities|Map|Keys (Variables|Default|GetDialogue%20Bumpers)))
  (return _keys))

(fn GetBumpers
  (params (Key :type String))
  (bind (_value _returnvalue) (Utilities|Map|Find (Variables|Default|GetDialogue%20Bumpers) (Utilities|String|StringToName Key)))
  (return _returnvalue _value))
