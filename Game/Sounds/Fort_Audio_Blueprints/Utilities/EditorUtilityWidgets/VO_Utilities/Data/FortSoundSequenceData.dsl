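; FortSoundSequenceData is a data object for the VO editor utilities, holding an ordered dialogue sound sequence with optional start and end bumpers.
; Add, remove and reorder functions maintain the dialogue array and row indexes, while setters edit sounds and post-sentence delays.
; SetDialogueBumpers looks up a key in a bumpers data asset, then updates, constructs or clears bumper entries.
; GetDialogueData returns the array wrapped with bumpers that have sounds.
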(fn AddDialogueData
  (params (Dialogue%20Data%20to%20Add :type Fort%20Dialogue%20Data%20Object%20Reference))
  (Variables|/Game/Sounds/Fort_Audio_Blueprints/Utilities/EditorUtilityWidgets/VO_Utilities/Data/FortDialogueData.FortDialogueData_C|SetRow%20Index (Utilities|Array|Length (Variables|Default|GetDialogue%20Data)) Dialogue%20Data%20to%20Add)
  (Utilities|Array|Add (Variables|Default|GetDialogue%20Data) (Variables|GetDialogue%20Data%20to%20Add)))

(fn GetDialogueData
  (Utilities|FlowControl|Sequence
    (:then_0
      (bind _dialogue_start_sound_data (Variables|Default|GetDialogue%20Start%20Sound%20Data)
        (:then
          (Variables|/Game/Sounds/Fort_Audio_Blueprints/Utilities/EditorUtilityWidgets/VO_Utilities/Data/FortDialogueData.FortDialogueData_C|GetSound _dialogue_start_sound_data
            (:then
              (Utilities|Array|Add (LocalVariables|GetLocal%20Dialogue%20Data) _dialogue_start_sound_data))
            (:else)))
        (:else)))
    (:then_1
      (Utilities|FlowControl|Sequence
        (:then_0
          (Utilities|Array|AppendArray (LocalVariables|GetLocal%20Dialogue%20Data) (Variables|Default|GetDialogue%20Data)))
        (:then_1
          (Utilities|FlowControl|Sequence
            (:then_0
              (bind _dialogue_end_sound_data (Variables|Default|GetDialogue%20End%20Sound%20Data)
                (:then
                  (Variables|/Game/Sounds/Fort_Audio_Blueprints/Utilities/EditorUtilityWidgets/VO_Utilities/Data/FortDialogueData.FortDialogueData_C|GetSound _dialogue_end_sound_data
                    (:then
                      (Utilities|Array|Add (LocalVariables|GetLocal%20Dialogue%20Data) _dialogue_end_sound_data))
                    (:else)))
                (:else)))
            (:then_1
              (return (LocalVariables|GetLocal%20Dialogue%20Data)))))))))

(fn RemoveDialogueData
  (params (Dialogue%20Data%20to%20Remove :type Fort%20Dialogue%20Data%20Object%20Reference))
  (bind _returnvalue (Utilities|Array|RemoveItem (Variables|Default|GetDialogue%20Data) Dialogue%20Data%20to%20Remove))
  (if _returnvalue
    (Default|ResetDialogueDataIndexes)
    (return true)
    (else
      (return false))))

(fn ReorderDialogueData
  (params (Dialogue%20Data%20to%20Insert :type Fort%20Dialogue%20Data%20Object%20Reference) (Dialogue%20Data%20to%20Insert%20Above :type Fort%20Dialogue%20Data%20Object%20Reference))
  (Utilities|FlowControl|Sequence
    (:then_0
      (Utilities|Array|RemoveItem (Variables|Default|GetDialogue%20Data) (Variables|GetDialogue%20Data%20to%20Insert)))
    (:then_1
      (Utilities|FlowControl|Sequence
        (:then_0
          (LocalVariables|SetIndex%20to%20Insert%20At (Utilities|Array|FindItem (Variables|Default|GetDialogue%20Data) (Variables|GetDialogue%20Data%20to%20Insert%20Above))))
        (:then_1
          (Utilities|FlowControl|Sequence
            (:then_0
              (if (>= (LocalVariables|GetIndex%20to%20Insert%20At) 0)
                (Utilities|Array|Insert (Variables|Default|GetDialogue%20Data) (Variables|GetDialogue%20Data%20to%20Insert) (LocalVariables|GetIndex%20to%20Insert%20At))
                (else
                  _))
              (Default|ResetDialogueDataIndexes))))))))

(fn ResetDialogueDataIndexes
  (for (_array_element _array_index) (Variables|Default|GetDialogue%20Data)
    (Variables|/Game/Sounds/Fort_Audio_Blueprints/Utilities/EditorUtilityWidgets/VO_Utilities/Data/FortDialogueData.FortDialogueData_C|SetRow%20Index _array_index _array_element)))

(fn SetDialogueBumpers
  (params (Key :type String) (Dialogue%20Bumpers%20Data%20Asset :type Fort%20Dialogue%20Bumpers%20Data%20Object%20Reference%20%28by%20ref%29))
  (bind (_dialoguestartsound_2_f11fb62c4ceec978498adba99a0a47b6 _dialoguestartpost_sentencedelay_7_c0201aa0427e675c8fe761ae67d48f99 _dialogueendsound_4_a332ef504c59d9f72a0661be991b47f9) (Utilities|Struct|BreakFortDialogueBumpers (.Bumpers (Class|FortDialogueBumpersData|GetBumpers Dialogue%20Bumpers%20Data%20Asset Key))))
  (bind _returnvalue (Utilities|Text|ToText(Object) _dialoguestartsound_2_f11fb62c4ceec978498adba99a0a47b6))
  (bind (_success _bumpers) (Class|FortDialogueBumpersData|GetBumpers Dialogue%20Bumpers%20Data%20Asset Key))
  (bind _returnvalue_2 (Utilities|Text|ToText(Object) _dialogueendsound_4_a332ef504c59d9f72a0661be991b47f9))
  (if _success
    (Variables|Default|SetDialogue%20Bumpers (Utilities|String|StringToName Key))
    (Utilities|FlowControl|Sequence
      (:then_0
        (Utilities|IsValid _dialoguestartsound_2_f11fb62c4ceec978498adba99a0a47b6
          (:Is%20Valid
            (bind _dialogue_start_sound_data (Variables|Default|GetDialogue%20Start%20Sound%20Data)
              (:then
                (Variables|/Game/Sounds/Fort_Audio_Blueprints/Utilities/EditorUtilityWidgets/VO_Utilities/Data/FortDialogueData.FortDialogueData_C|SetSound _dialoguestartsound_2_f11fb62c4ceec978498adba99a0a47b6 _dialogue_start_sound_data)
                (Variables|/Game/Sounds/Fort_Audio_Blueprints/Utilities/EditorUtilityWidgets/VO_Utilities/Data/FortDialogueData.FortDialogueData_C|SetSpoken%20Text _returnvalue _dialogue_start_sound_data))
              (:else
                (bind _returnvalue_3 (Game|ConstructObjectfromClass "/Game/Sounds/Fort_Audio_Blueprints/Utilities/EditorUtilityWidgets/VO_Utilities/Data/FortDialogueData.FortDialogueData_C" 0 0 (Variables|Default|GetAsset%20Name) "None" "Start Sound" _returnvalue "Start Sound" _dialoguestartsound_2_f11fb62c4ceec978498adba99a0a47b6 _dialoguestartpost_sentencedelay_7_c0201aa0427e675c8fe761ae67d48f99 false))
                (Variables|Default|SetDialogue%20Start%20Sound%20Data _returnvalue_3))))
          (:Is%20Not%20Valid
            (Variables|Default|SetDialogue%20Start%20Sound%20Data 0))))
      (:then_1
        (Utilities|FlowControl|Sequence
          (:then_0
            (Utilities|IsValid _dialogueendsound_4_a332ef504c59d9f72a0661be991b47f9
              (:Is%20Valid
                (bind _dialogue_end_sound_data (Variables|Default|GetDialogue%20End%20Sound%20Data)
                  (:then
                    (Variables|/Game/Sounds/Fort_Audio_Blueprints/Utilities/EditorUtilityWidgets/VO_Utilities/Data/FortDialogueData.FortDialogueData_C|SetSound _dialogueendsound_4_a332ef504c59d9f72a0661be991b47f9 _dialogue_end_sound_data)
                    (Variables|/Game/Sounds/Fort_Audio_Blueprints/Utilities/EditorUtilityWidgets/VO_Utilities/Data/FortDialogueData.FortDialogueData_C|SetSpoken%20Text _returnvalue_2 _dialogue_end_sound_data))
                  (:else
                    (bind _returnvalue_4 (Game|ConstructObjectfromClass "/Game/Sounds/Fort_Audio_Blueprints/Utilities/EditorUtilityWidgets/VO_Utilities/Data/FortDialogueData.FortDialogueData_C" 0 0 (Variables|Default|GetAsset%20Name) "None" "Start Sound" _returnvalue_2 "Start Sound" _dialogueendsound_4_a332ef504c59d9f72a0661be991b47f9 0.0 false))
                    (Variables|Default|SetDialogue%20End%20Sound%20Data _returnvalue_4))))
              (:Is%20Not%20Valid
                (Variables|Default|SetDialogue%20End%20Sound%20Data 0)))))))
    (else
      (Variables|Default|SetDialogue%20Start%20Sound%20Data 0)
      (Variables|Default|SetDialogue%20End%20Sound%20Data 0))))

(fn SetDialoguePost-sentenceDelayAmount
  (params (Dialogue%20Data :type Fort%20Dialogue%20Data%20Object%20Reference) (Delay%20Amount :type Float%20%28double-precision%29))
  (bind _dialogue_data (Variables|Default|GetDialogue%20Data))
  (bind _dialogue_data_2 (Variables|Default|GetDialogue%20Data))
  (bind _returnvalue (Utilities|Array|FindItem _dialogue_data_2 _dialogue_data))
  (bind _dialogue_data_3 (Variables|Default|GetDialogue%20Data))
  (bind _output (Utilities|Array|Get(acopy) _dialogue_data_3 _returnvalue))
  (if (>= _returnvalue 0)
    (Utilities|IsValid _output
      (:Is%20Valid
        (Variables|/Game/Sounds/Fort_Audio_Blueprints/Utilities/EditorUtilityWidgets/VO_Utilities/Data/FortDialogueData.FortDialogueData_C|SetPost-Sentence%20Delay (Variables|GetDelay%20Amount) _output)))
    (elif (== (Variables|Default|GetDialogue%20Data) (Variables|Default|GetDialogue%20Start%20Sound%20Data))
      (bind _dialogue_start_sound_data (Variables|Default|GetDialogue%20Start%20Sound%20Data)
        (:then
          (Variables|/Game/Sounds/Fort_Audio_Blueprints/Utilities/EditorUtilityWidgets/VO_Utilities/Data/FortDialogueData.FortDialogueData_C|SetPost-Sentence%20Delay (Variables|GetDelay%20Amount) _dialogue_start_sound_data))
        (:else))
      (elif (== (Variables|Default|GetDialogue%20Data) (Variables|Default|GetDialogue%20End%20Sound%20Data))
        (bind _dialogue_end_sound_data (Variables|Default|GetDialogue%20End%20Sound%20Data)
          (:then
            (Variables|/Game/Sounds/Fort_Audio_Blueprints/Utilities/EditorUtilityWidgets/VO_Utilities/Data/FortDialogueData.FortDialogueData_C|SetPost-Sentence%20Delay (Variables|GetDelay%20Amount) _dialogue_end_sound_data))
          (:else))))))

(fn SetDialogueSound
  (params (Dialogue%20Data :type Fort%20Dialogue%20Data%20Object%20Reference) (Sound :type Sound%20Base%20Object%20Reference))
  (bind _dialogue_data (Variables|Default|GetDialogue%20Data))
  (bind _dialogue_data_2 (Variables|Default|GetDialogue%20Data))
  (bind _returnvalue (Utilities|Array|FindItem _dialogue_data_2 _dialogue_data))
  (bind _dialogue_data_3 (Variables|Default|GetDialogue%20Data))
  (bind _output (Utilities|Array|Get(acopy) _dialogue_data_3 _returnvalue))
  (if (>= _returnvalue 0)
    (Utilities|IsValid _output
      (:Is%20Valid
        (Variables|/Game/Sounds/Fort_Audio_Blueprints/Utilities/EditorUtilityWidgets/VO_Utilities/Data/FortDialogueData.FortDialogueData_C|SetSound (Variables|GetSound) _output)))))
